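# Translucent unlit Niagara sprite material for rift smoke in the SunShine pickaxe effects.
# Particle color drives emissive through a shared adjustment function.
# Two noise samples, tiled and panned on world-space-aligned screen coordinates, are multiplied and square-rooted to break up the smoke independently of sprite UVs.
# That noise, plus a faint mask term, scales smoke-mask and particle alpha, with depth fade softening intersections.
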Material /Game/Weapons/FORT_Melee/Pickaxe_SunShine/FX/Flake/M_Rift_Smoke
graph {
 properties {bUsedWithNiagaraSprites=true, blendMode=BLEND_Translucent, pixelDepthOffsetMode=PDOM_Legacy, shadingModel=MSM_Unlit, usedShadingModels=Unlit}
 nodes [
  MaterialFunctionCall(Emissive=ParticleColor().RGB) {materialFunction=&MF_FX_EmissiveAdjustments}
  MaterialFunctionCall!(
   UV=Divide(
    A=MaterialFunctionCall!(Center Point World Position) {materialFunction=/Engine/Functions/Engine_MaterialFunctions02/UVs/WorldSpaceAlignedScreenCoordinates}.Offset UVs
    B=ScalarParameter() {defaultValue=80}
   )
   Tile U=ScalarParameter() {defaultValue=0.5, parameterName="Tile U (S)_1"}
   Tile V=ScalarParameter() {defaultValue=0.2, parameterName="Tile V (S)"}
   Pan U
   Pan V=ScalarParameter() {defaultValue=-0.2, parameterName="Pan V (S)"}
   Offset U
   Offset V
   Time
  ) {materialFunction=/Game/Effects/Fort_Effects/Materials/_Functions/MF_PanTileOffset}
  n1 = TextureSample() {texture=/Game/Effects/Fort_Effects/Textures/T_Smooth_Smoke_mask}
  n2 = ParticleColor()
  Multiply(A=@n1.RGB, B=@n2.RGB)
  n3 = Multiply(A=@n1.A) {constB=0.1}
  n4 = ScalarParameter() {defaultValue=1, parameterName="Tile U2 (S)"}
  n5 = MaterialFunctionCall(
   UV=Divide(
    A=MaterialFunctionCall() {materialFunction=&WorldSpaceAlignedScreenCoordinates}.Offset UVs
    B=ScalarParameter() {defaultValue=80, parameterName=NoiseScale}
   )
   Tile U=@n4
   Tile V=@n4
   Pan V=ScalarParameter() {defaultValue=-0.2, parameterName="Pan V 2(S)"}
  ) {materialFunction=&MF_PanTileOffset}
  n6 = SquareRoot(None=Multiply(
    A=TextureSample!(UVs, Tex, Level, Apply View MipBias) {mipValueMode=TMVM_MipLevel, texture=&/CINE_T_SimpleNoise}.G
    B=TextureSample(UVs=@n5.UV's) {mipValueMode=TMVM_MipLevel, texture=&CINE_T_SimpleNoise}.G
   ))
  Multiply(A=Multiply(A=Multiply(A=@n1.A, B=@n2.A), B=Add(A=@n3, B=@n6)), B=DepthFade() {fadeDistanceDefault=60})
 ]
}
